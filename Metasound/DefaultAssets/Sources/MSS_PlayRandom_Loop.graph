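# MSS_PlayRandom_Loop is a reusable looping stereo MetaSound Source that plays a random wave from its Sounds array on play.
# A shared-state random get, with no-repeat count derived from array length, triggers the wave player.
# OnPlay also randomizes pitch and volume symmetrically around their base inputs, using the spread inputs.
# Volume is clamped in decibels, converted to linear gain and applied to both channels.
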MetaSoundSource /Metasound/DefaultAssets/Sources/MSS_PlayRandom_Loop
inputs [
 PitchBase {type=Float}
 PitchRandomSpread {type=Float}
 Sounds {type=WaveAsset:Array}
 UE.Source.OnPlay {type=Trigger}
 VolumeBase {type=Float}
 VolumeRandomSpread {type=Float}
]
interfaces [UE.OutputFormat.Stereo:1.0, UE.Source:1.1]
outputs [
 "UE.OutputFormat.Stereo.Audio:0" {type=Audio}
 "UE.OutputFormat.Stereo.Audio:1" {type=Audio}
]
pages [Default {
  input_defaults {PitchBase=0, PitchRandomSpread=0, UE.Source.OnPlay=null, VolumeBase=0, VolumeRandomSpread=0, Sounds []}
  nodes [
   n1 = UE.RandomFloat(
    Max=^VolumeRandomSpread
    Min=UE.Multiply:Float(PrimaryOperand=^VolumeRandomSpread) {also {defaults {AdditionalOperands=-1}}}.Out
    Next=UE.Reroute(Value=^UE.Source.OnPlay).Value
   )
   DecibelstoLinearGain_0 = UE.Decibels to Linear Gain:Float(Decibels=Clamp.Clamp:Float(In=UE.Add:Float(PrimaryOperand=^VolumeBase, AdditionalOperands=@n1.Value).Out) {also {defaults {Max=12, Min=-60}}}.Value)
   n2 = UE.Multiply:Float(PrimaryOperand=^PitchRandomSpread) {also {defaults {AdditionalOperands=-1}}}
   n3 = UE.Reroute(Value=^UE.Source.OnPlay)
   ArrayRandomGet_0 = Array.Random Get:WaveAsset:Array(In Array=^Sounds, Next=^UE.Source.OnPlay, No Repeats=UE.Divide:Int32(PrimaryOperand=Array.Num:WaveAsset:Array(Array=^Sounds).Num).Out) {also {defaults {Enable Shared State="<MetaSoundBuilderResult.SUCCEEDED: 0>", Shared State Behavior=2}}}
   WavePlayer_0 = UE.Wave Player:Stereo(
    Pitch Shift=UE.Add:Float(PrimaryOperand=^PitchBase, AdditionalOperands=UE.RandomFloat(Max=^PitchRandomSpread, Min=@n2.Out, Next=@n3.Value).Value).Out
    Play=@ArrayRandomGet_0.On Next
    Wave Asset=@ArrayRandomGet_0.Value
   ) {also {defaults {Loop="<MetaSoundBuilderResult.SUCCEEDED: 0>"}, version [1,1]}}
   Multiply_0 = UE.Multiply:Audio by Float(AdditionalOperands=@DecibelstoLinearGain_0.Linear Gain, PrimaryOperand=@WavePlayer_0.Out Left)
   Multiply_1 = UE.Multiply:Audio by Float(AdditionalOperands=@DecibelstoLinearGain_0.Linear Gain, PrimaryOperand=@WavePlayer_0.Out Right)
  ]
  links [{dst="@UE.OutputFormat.Stereo.Audio:0", src=Multiply_0.Out}, {dst="@UE.OutputFormat.Stereo.Audio:1", src=Multiply_1.Out}]
 }]
properties {
 assetImportData=&self:AssetImportData
 bLooping=true
 duration=10000
 numChannels=2
 outputFormat=Stereo
 sampleRate=48000
}
